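# Level Sequence for the idle Niagara effect of the CampWrench pickaxe cosmetic, apparently for previewing or timing that effect.
# It possesses the NS_Pickaxe_CampWrench_Idle Niagara actor and binds its Niagara component as a child.
# The component carries a single Niagara system track whose spawn section controls the system's life cycle over a short range.
# There are no other tracks, so the sequence animates nothing else.
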LevelSequence /BRCosmetics/Weapons/FORT_Melee/Pickaxe_CampWrenchUniversal/FX/CampWrench_FX_Sequence
bindings [
 NiagaraComponent0 {
  id=#1
  parent_id=#2
  possessed_class=NiagaraComponent
  tracks [{type=MovieSceneNiagaraSystemTrack, sections [{type=MovieSceneNiagaraSystemSpawnSection, end_frame=300, start_frame=0}]}]
 }
 NS_Pickaxe_CampWrench_Idle {id=#2, possessed_class=NiagaraActor, tracks []}
]
display_rate {denominator=1, numerator=30}
tick_resolution {denominator=1, numerator=24000}
tracks []
