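# Translucent Niagara mesh-particle material for the sphere effect on the CampWrench pickaxe, with an animated, iridescent energy look.
# Panning mask textures warp position-based UVs with a BumpOffset, feeding thin-film spectrum color that blends to a Fresnel edge tint, scaled by particle color.
# Opacity combines Fresnel, particle alpha, view-angle fade and depth fade (default and deferred paths only).
# Dynamic parameters drive Fresnel exponent, an erosion mask and noise-based vertex offset along normals.
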Material /BRCosmetics/Weapons/FORT_Melee/Pickaxe_CampWrenchUniversal/FX/M_CampWrench_Sphere
graph {
 properties {bAutomaticallySetUsageInEditor=false, bUsedWithNiagaraMeshParticles=true, blendMode=BLEND_Translucent, usedShadingModels=Default Lit}
 nodes [
  n1 = TextureCoordinate()
  n2 = Add(
   A=MaterialFunctionCall(UVs=@n1, CenterPosition=Constant2Vector() {g=0.5, r=0.5}, Radius=Constant() {r=0.15}, Density=Constant() {r=1}) {materialFunction=&RadialGradientExponential}.RadialGradientExponential
   B=MaterialFunctionCall(UVs=TextureCoordinate(), CenterPosition=Constant2Vector() {g=0.5, r=0.25}, Radius=Constant() {r=0.15}, Density=Constant() {r=1}) {materialFunction=&RadialGradientExponential}.RadialGradientExponential
  )
  n3 = TextureSample(UVs=Panner(Coordinate=TextureCoordinate() {uTiling=5, vTiling=5}) {speedX=0.2, speedY=0.251}) {
   samplerType=SAMPLERTYPE_LinearColor
   texture=/BRCosmetics/Characters/Player/Male/Medium/Bodies/M_MED_EggnogFaucet/Materials/T_EggnogFaucet_mask
  }
  Add(A=Multiply(A=Multiply(A=@n2, B=@n3.B)) {constB=25}, B=@n3.B)
  ComponentMask(None=LocalPosition().XYZ) {r=true}
  ComponentMask() {g=true, r=true}
  Constant() {r=1}
  n4 = Subtract(
   A=WorldPosition().XYZ
   B=MaterialFunctionCall() {materialFunction=&ComponentPivotLocation}.Result
  )
  n5 = Saturate(None=Add(A=Divide(A=@n4) {constB=100}) {constB=0.5})
  Divide(A=Add(A=OneMinus(None=@n5), B=NamedRerouteUsage()))
  MaterialFunctionCall(Emissive=NamedRerouteUsage()) {materialFunction=&MF_FX_EmissiveAdjustments}
  n6 = NamedRerouteUsage()
  n7 = Add(
   A=RotateAboutAxis(NormalizedRotationAxis=Constant3Vector() {constant=<rgba 0 1 0 0>}, RotationAngle=Constant() {r=0.25}, PivotPoint=Constant(), Position=@n6)
   B=@n6
  )
  n8 = ComponentMask(None=@n7) {b=true, r=true}
  n9 = TextureSample(UVs=Panner(Coordinate=Multiply(A=@n8) {constB=0.5}) {speedX=0.05, speedY=-0.05}) {
   samplerType=SAMPLERTYPE_LinearColor
   texture=&T_EggnogFaucet_mask
  }
  n10 = Time()
  n11 = Sine(None=Add(A=@n9.B, B=Multiply(A=@n10) {constB=0.3}))
  n12 = Time()
  n13 = Constant3Vector() {constant=<rgba -1 0 0.5 0>}
  n14 = Add(
   A=Add(A=@n7, B=NamedRerouteDeclaration(None=@n11) {name=AnimatedGaussian_1, nodeColor=<rgba 0 1 0.6016 1>})
   B=Multiply(A=Multiply(A=@n12, B=@n13)) {constB=0.15}
  )
  n15 = Multiply(A=ComponentMask(None=@n14) {b=true, r=true}) {constB=0.75}
  n16 = Constant2Vector() {g=51.2135, r=11.213}
  n17 = TextureSample(UVs=BumpOffset(Coordinate=Add(A=@n15, B=@n16), Height=@n9.G) {heightRatio=0.5}) {
   samplerType=SAMPLERTYPE_LinearColor
   texture=&T_EggnogFaucet_mask
  }
  Multiply(A=@n17.RGB)
  Multiply(A=NamedRerouteUsage(), B=NamedRerouteUsage())
  n18 = ComponentMask(None=@n1) {g=true}
  n19 = Multiply(A=Multiply(A=Power(Base=@n18) {constExponent=3}) {constB=2}, B=Multiply(A=@n3.B) {constB=0.5})
  n20 = Power(Base=@n18) {constExponent=12}
  n21 = DynamicParameter() {defaultValue=<rgba 0 0 1 0>, paramNames [Erosion, Use Particle Color, Fresnel Exp, WPO]}
  n22 = ScalarParameter() {parameterName=Erosion Debug}
  NamedRerouteDeclaration(None=MaterialFunctionCall!(
    Gradient=Add(A=@n19, B=@n20)
    Erosion Amount=Add(A=@n21.Erosion, B=@n22)
    Softness=ScalarParameter() {defaultValue=0.05, parameterName=Erosion Softness, sliderMax=1}
   ) {materialFunction=/BRCosmetics/Athena/Effects/Particles/Skydiving/NS_Contrail_DarkAndLight/Functions/MF_Erosion}.Increased Range) {name=Erosion, nodeColor=<rgba 1 0.4688 0 1>}
  n23 = NamedRerouteDeclaration(None=@n17.G) {name=DistortedVoron, nodeColor=<rgba 1 0 0.9609 1>}
  n24 = Multiply(A=Time()) {constB=0.5}
  NamedRerouteDeclaration(None=Sine(None=Add(A=@n23, B=@n24))) {name=DistortedVoronAnim, nodeColor=<rgba 0 0.7422 1 1>}
  n25 = ConstantBiasScale(None=NamedRerouteUsage())
  n26 = NamedRerouteUsage()
  n27 = LinearInterpolate(Alpha=Saturate(None=Multiply(A=@n25, B=@n26))) {constA=0.2, constB=0.8}
  n28 = Time()
  n29 = Sine(None=Add(A=NamedRerouteUsage(), B=Multiply(A=@n28) {constB=0.15}))
  n30 = NamedRerouteUsage()
  n31 = TextureSample(UVs=Multiply(A=Multiply(A=@n29, B=@n30)) {constB=0.5}) {texture=/Game/Global/Textures/Default/Thinfilm/T_ThinFilm_Spectrum_COLOR}
  n32 = DynamicParameter() {defaultValue=<rgba 0 0 1 0>, paramNames [Erosion, Use Particle Color, Fresnel Exp, WPO]}
  n33 = ScalarParameter() {defaultValue=0.1, parameterName=Fresnel_BRF}
  n34 = Saturate(None=Power(Base=Fresnel(ExponentIn=@n32.Fresnel Exp, BaseReflectFractionIn=@n33) {baseReflectFraction=0.1, exponent=1}) {constExponent=3})
  n35 = LinearInterpolate(
   A=Desaturation(None=Multiply(A=@n27, B=@n31.RGB), Fraction=ScalarParameter() {defaultValue=-0.75, parameterName=FurtacorIntensity})
   B=VectorParameter() {parameterName=EdgeColor, defaultValue=<rgba 0.2618 0.6945 1 1>}.RGB
   Alpha=@n34
  ) {constAlpha=0.2}
  NamedRerouteDeclaration(None=Multiply(A=ParticleColor().RGB, B=Multiply(A=@n35) {constB=15})) {name=OutEmissive, nodeColor=<rgba 0 1 0.7188 1>}
  n36 = Multiply(A=ParticleColor().A, B=@n34)
  n37 = Sine(None=ComponentMask(None=TextureCoordinate() {vTiling=0.5}) {g=true})
  n38 = Saturate(None=Power(Base=Multiply(A=@n37) {constB=1.3}) {constExponent=1})
  n39 = Multiply(
   A=LinearInterpolate(A=@n36, B=Multiply(A=@n36, B=@n38), Alpha=ScalarParameter() {parameterName=Use_LinearMask})
   B=MaterialFunctionCall("FadeSensitivity (optional)"=ScalarParameter() {defaultValue=1.5, parameterName=MeshFade}) {materialFunction=&FadeBasedOnViewAngle}.Result
  )
  n40 = DepthFade(Opacity=@n39, FadeDistance=ScalarParameter() {defaultValue=5, parameterName=DepthFade})
  NamedRerouteDeclaration(None=ShadingPathSwitch(Default=@n40, Deferred=@n40, Forward=@n39, Mobile=@n39)) {name=Opacity, nodeColor=<rgba 0.4531 1 0 1>}
  NamedRerouteDeclaration(None=@n17.B) {name=DistortedGaussian, nodeColor=<rgba 1 0 0.1641 1>}
  n41 = VertexNormalWS()
  n42 = TextureCoordinate()
  n43 = VectorParameter() {parameterName=WPO UV Controls, channelNames {a=Speed V, b=Speed U, g=Tile V, r=Tile U}, defaultValue=<rgba 3 3 0.5 0.5>}
  n44 = AppendVector(A=@n43.Tile U, B=@n43.Tile U)
  n45 = TextureSample(UVs=Panner(Coordinate=Multiply(A=@n42, B=@n44) {constB=3}, Speed=AppendVector(A=@n43.Speed U, B=@n43.Speed U)) {speedX=0.5, speedY=0.5}) {samplerType=SAMPLERTYPE_LinearColor, texture=&HighContrastNoise}
  NamedRerouteDeclaration(None=Multiply(
    A=Multiply(A=@n41, B=@n45.RGB)
    B=DynamicParameter() {defaultValue=<rgba 0 0 1 0>, paramNames [Erosion, Use Particle Color, Fresnel Exp, WPO]}.WPO
   )) {name=WPO, nodeColor=<rgba 0.0781 0 1 1>}
  n46 = TextureObject() {
   samplerType=SAMPLERTYPE_LinearColor
   texture=&T_EggnogFaucet_mask
  }
  n47 = PreSkinnedNormal()
  n48 = MaterialFunctionCall(Color=MaterialFunctionCall(TextureObject=@n46, TextureSize=Constant() {r=2}, WorldPosition=@n14, World Space Normal=VertexInterpolator(VS=@n47).PS) {materialFunction=&WorldAlignedTexture}.XYZ Texture) {materialFunction=&SplitComponents}
  NamedRerouteDeclaration(None=@n48.G) {name=DistortedVoron_1, nodeColor=<rgba 1 0 0.9609 1>}
  NamedRerouteDeclaration(None=@n48.B) {name=DistortedGaussian_1, nodeColor=<rgba 1 0 0.1641 1>}
  n49 = Multiply(A=LocalPosition() {localOrigin=InstancePreSkinning}.XYZ) {constB=0.1}
  NamedRerouteDeclaration(None=ShaderStageSwitch(PixelShader=VertexInterpolator(VS=@n49).PS, VertexShader=@n49)) {name=PreSkinnedPosition, nodeColor=<rgba 0.6406 1 0 1>}
  NamedRerouteUsage()
  ScalarParameter() {defaultValue=1, parameterName=Fresnel_Exp}
  n50 = NamedRerouteUsage()
  n51 = Multiply(A=MaterialFunctionCall(UVs=ComponentMask(None=@n50) {b=true, r=true}, "Rotation Angle (0-1)"=Constant() {r=0.25}) {materialFunction=&CustomRotator}.Rotated Values) {constB=0.55}
  n52 = TextureSample(UVs=Panner(Coordinate=@n51) {speedX=0.05, speedY=-0.05}) {
   samplerType=SAMPLERTYPE_LinearColor
   texture=&T_EggnogFaucet_mask
  }
  n53 = Add(A=@n52.B, B=Multiply(A=Time()) {constB=0.3})
  n54 = Multiply(A=NamedRerouteDeclaration(None=Sine(None=@n53)) {name=AnimatedGaussian, nodeColor=<rgba 0 1 0.6016 1>}) {constB=0.05}
  TextureSample(UVs=Panner(Coordinate=Add(A=@n51, B=@n54)) {speedX=0.05, speedY=0.15}) {
   samplerType=SAMPLERTYPE_LinearColor
   texture=&T_EggnogFaucet_mask
  }
 ]
}
