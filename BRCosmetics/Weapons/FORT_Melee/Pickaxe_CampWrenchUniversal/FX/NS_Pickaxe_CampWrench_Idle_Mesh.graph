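# Niagara idle effect for the CampWrench pickaxe: a glowing, mostly blue energy sphere attached to the weapon's skeletal mesh.
# Delayed gathering and burst emitters lead into looping spheres, chromatic orbiters, glow sprites, meshes, a light and pulsing socket rings.
# Multi-element emitters partition particles to give each its own color, size, curves and visibility tag.
# User.BuiltIn blends positions off the sampled mesh location and zeroes ring spawns.
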NiagaraSystem /BRCosmetics/Weapons/FORT_Melee/Pickaxe_CampWrenchUniversal/FX/NS_Pickaxe_CampWrench_Idle_Mesh
dependencies {
 usedDataInterfaces [
  &NiagaraDataInterfaceCurve
  &NiagaraDataInterfaceColorCurve
  &NiagaraDataInterfaceArrayInt32
  &NiagaraDataInterfaceSkeletalMesh
 ]
 usedDynamicInputs [&SimulationPosition, &Vector2DFromFloat]
 usedModules [
  &self:SystemSpawnScript.SystemScriptSource.SystemScriptGraph.NiagaraNodeAssignment_4.SetVariables_#1
  &self:SystemSpawnScript.SystemScriptSource.SystemScriptGraph.NiagaraNodeAssignment_2.SetVariables_#2
  &self:SystemSpawnScript.SystemScriptSource.SystemScriptGraph.NiagaraNodeAssignment_6.SetVariables_#3
  &self:SystemSpawnScript.SystemScriptSource.SystemScriptGraph.NiagaraNodeAssignment_8.SetVariables_#4
  &SystemState
  &self:SystemSpawnScript.SystemScriptSource.SystemScriptGraph.NiagaraNodeAssignment_5.SetVariables_#5
 ]
}
emitters {
 BurstElements {
  emitterUpdateScript [
   &EmitterState(
    Inactive Response=::NewEnumerator1
    Life Cycle Mode=::NewEnumerator1
    Loop Behavior=::NewEnumerator1
    Loop Delay=$System.Delay
    Loop Duration=2
    UseLoopDelay=true
   )
   &SpawnBurst_Instantaneous(Spawn Count=2, Spawn Group=1, Spawn Time=0.12, Use Spawn Probability=false, UseLoopCountLimit=false) {module=SpawnBurst_Instantaneous001}
   &SpawnBurst_Instantaneous(Spawn Count=2, Spawn Time=0.2, Use Spawn Probability=false, UseLoopCountLimit=false) {module=SpawnBurst_Instantaneous002}
  ]
  particleSpawnScript [
   &PartitionParticles(Number Of Partitions=4)
   /SetVariables(Particles.VisibilityTag=$Particles.PartitionParticles.Partition) {isSetParameters=true, module=SetVariables_#6}
   &InitializeParticle(
    Color=&MultiplyLinearColor(
     Linear Color=&MergeHelper_ColorSwitch(
      0=<rgba 2 2 2 1>
      1=<rgba 3 3 3 1>
      2=<rgba 3 3 3 2>
      3=<rgba 0.6284 0.6284 5 1>
      Index=$Particles.PartitionParticles.Partition
      Num=::NewEnumerator2
     )
     Scale Factor=&MakeVector4_FromLinearColor(LinearColor=$System.TODColorMultiplier)
    )
    Lifetime=&MergeHelper_FloatSwitch(0=0.35, 1=0.35, 2=0.4, 3=0.4, Index=$Particles.PartitionParticles.Partition, Num=::NewEnumerator2)
    Position Offset=$System.PositionOffset
    Sprite Rotation Mode=::NewEnumerator2
    Sprite Size Mode=::NewEnumerator3
    Uniform Sprite Size=&Multiply_Float(
     A=&MergeHelper_FloatSwitch(0=45, 1=45, 2=45, 3=45, Index=$Particles.PartitionParticles.Partition, Num=::NewEnumerator2)
     B=$System.GlobalScale
    )
    UsePositionOffset=true
    Write Material Random=true
   ) {module=InitializeParticle001}
   &CameraOffset(Camera Offset Amount=&MergeHelper_FloatSwitch(0=-5, 1=-5, 2=5, 3=5, Index=$Particles.PartitionParticles.Partition, Num=::NewEnumerator2))
  ]
  particleUpdateScript [
   &ParticleState()
   &SkeletalMeshLocation(
    MaskSpawnGroup=false
    Mesh Sampling Type=::NewEnumerator5
    Position Sampling=::NewEnumerator2
    Skeletal Mesh=$System.Skeletal Mesh
    Socket Sampling Mode=::NewEnumerator5
   )
   /SetVariables(Particles.Position=&Lerp_Vector(
     A=Linked() {linkedVariable {name=Particles.SkeletalMeshLocation.SampledPosition, type=NiagaraPosition}}
     B=Linked() {linkedVariable {name=Particles.Position, type=NiagaraPosition}}
     Alpha=$User.BuiltIn
    )) {isSetParameters=true, module=SetVariables_#7}
   &SolveForcesAndVelocity(Clamp Velocity=false, Limit Acceleration=false)
   &ScaleColor(
    Scale Alpha=&MergeHelper_FloatSwitch(
     0=&FloatFromCurve(
      FloatCurve=DataInterface() {propertyValues {Curve {keys [
          {arriveTangent=6.4019, interpMode=RCIM_Cubic, leaveTangent=6.4019, tangentMode=RCTM_User, time=0, value=0}
          {arriveTangent=-0.2788, interpMode=RCIM_Cubic, leaveTangent=-0.2788, tangentMode=RCTM_User, time=0.2, value=0.7}
          {arriveTangent=-0.021, interpMode=RCIM_Cubic, leaveTangent=-0.021, tangentMode=RCTM_User, time=1, value=0}
         ]}}}
      Scale Curve=2
     )
     1=&FloatFromCurve(
      FloatCurve=DataInterface() {propertyValues {Curve {keys [
          {arriveTangent=6.4019, interpMode=RCIM_Cubic, leaveTangent=6.4019, tangentMode=RCTM_User, time=0, value=0}
          {arriveTangent=-0.2788, interpMode=RCIM_Cubic, leaveTangent=-0.2788, tangentMode=RCTM_User, time=0.2, value=0.7}
          {arriveTangent=-0.021, interpMode=RCIM_Cubic, leaveTangent=-0.021, tangentMode=RCTM_User, time=1, value=0}
         ]}}}
      Scale Curve=2
     )
     2=&FloatFromCurve(FloatCurve=DataInterface() {propertyValues {Curve {keys [
          {arriveTangent=-2.878, interpMode=RCIM_Cubic, leaveTangent=-2.878, tangentMode=RCTM_User, time=0, value=1}
          {arriveTangent=-1, interpMode=RCIM_Cubic, leaveTangent=-1, time=0.3, value=0.3}
          {interpMode=RCIM_Cubic, tangentMode=RCTM_User, time=1, value=0}
         ]}}})
     3=&FloatFromCurve(FloatCurve=DataInterface() {propertyValues {Curve {keys [
          {arriveTangent=-1.6399, interpMode=RCIM_Cubic, leaveTangent=-1.6399, tangentMode=RCTM_User, time=0, value=1}
          {interpMode=RCIM_Cubic, tangentMode=RCTM_User, time=1, value=0}
         ]}}})
     Index=$Particles.PartitionParticles.Partition
     Num=::NewEnumerator2
    )
    Scale RGB=&MakeVectorFromLinearColorRGB(Color=&MergeHelper_ColorSwitch(
      0=<rgba 1 1 1 1>
      1=<rgba 1 1 1 1>
      2=&ColorFromCurve(ColorCurve=DataInterface() {propertyValues {
         AlphaCurve {keys []}
         BlueCurve {keys [<at 0.1259 2 RCIM_Cubic>, <at 0.2277 5>, {interpMode=RCIM_Cubic, leaveTangent=-0.7175, time=0.4478, value=2}]}
         GreenCurve {keys [<at 0.1259 1.3894 RCIM_Cubic>, <at 0.2277 3.0143>, {interpMode=RCIM_Cubic, leaveTangent=-4.4353, time=0.4478, value=0.9149}]}
         RedCurve {keys [<at 0.1259 0.1681 RCIM_Cubic>, <at 0.2277 0.2344>, {interpMode=RCIM_Cubic, leaveTangent=2.2176, time=0.4478, value=0.1399}]}
        }})
      3=<rgba 1 1 1 1>
      Index=$Particles.PartitionParticles.Partition
      Num=::NewEnumerator2
     ))
    ScaleA=true
    ScaleRGB=true
   )
   &ScaleSpriteSize(
    Scale Sprite Size Mode=::NewEnumerator0
    Uniform Scale Factor=&MergeHelper_FloatSwitch(
     0=&FloatFromCurve(FloatCurve=DataInterface() {propertyValues {Curve {
         defaultValue=0
         keys [{interpMode=RCIM_Cubic, tangentMode=RCTM_User, time=0, value=0}, {interpMode=RCIM_Cubic, tangentMode=RCTM_User, time=1, value=1}]
        }}})
     1=&FloatFromCurve(FloatCurve=DataInterface() {propertyValues {Curve {
         defaultValue=0
         keys [{interpMode=RCIM_Cubic, tangentMode=RCTM_User, time=0, value=0}, {interpMode=RCIM_Cubic, tangentMode=RCTM_User, time=1, value=1}]
        }}})
     2=&FloatFromCurve(FloatCurve=DataInterface() {propertyValues {Curve {keys [
          {arriveTangent=6.5767, interpMode=RCIM_Cubic, leaveTangent=6.5767, tangentMode=RCTM_User, time=0, value=0}
          {arriveTangent=0.7192, interpMode=RCIM_Cubic, leaveTangent=0.7192, tangentMode=RCTM_User, time=0.2305, value=0.7836}
          {tangentMode=RCTM_User, time=1, value=1}
         ]}}})
     3=&FloatFromCurve(FloatCurve=DataInterface() {propertyValues {Curve {keys [
          {arriveTangent=6.5767, interpMode=RCIM_Cubic, leaveTangent=6.5767, tangentMode=RCTM_User, time=0, value=0}
          {arriveTangent=0.7192, interpMode=RCIM_Cubic, leaveTangent=0.7192, tangentMode=RCTM_User, time=0.2305, value=0.7836}
          {tangentMode=RCTM_User, time=1, value=1}
         ]}}})
     Index=$Particles.PartitionParticles.Partition
     Num=::NewEnumerator2
    )
   )
   &SpriteRotationRate(Rotation Rate=&UniformRangedFloat(Minimum=-50, Maximum=50))
  ]
  rendererClasses [&NiagaraSpriteRendererProperties]
  renderers [
   {rendererClass=&NiagaraSpriteRendererProperties}
   {rendererClass=&NiagaraSpriteRendererProperties, rendererIndex=1}
   {rendererClass=&NiagaraSpriteRendererProperties, rendererIndex=2}
   {rendererClass=&NiagaraSpriteRendererProperties, rendererIndex=3}
  ]
 }
 BurstRing {
  emitterUpdateScript [
   &EmitterState(
    Inactive Response=::NewEnumerator1
    Life Cycle Mode=::NewEnumerator1
    Loop Behavior=::NewEnumerator1
    Loop Delay=0.26
    Loop Duration Mode=::NewEnumerator1
    UseLoopDelay=true
   )
   &SpawnBurst_Instantaneous(
    Spawn Count=&MakeIntFromFloat(Float=&Lerp_Float(A=2, B=0, Alpha=$User.BuiltIn))
    Use Spawn Probability=false
    UseLoopCountLimit=false
   )
  ]
  particleSpawnScript [
   &PartitionParticles()
   &InitializeParticle(
    Color=&MultiplyLinearColor(
     Linear Color=<rgba 0.5 5 4.625 1>
     Scale Factor=&MakeVector4_FromLinearColor(LinearColor=$System.TODColorMultiplier)
    )
    Mesh Scale Mode=::NewEnumerator3
    Mesh Uniform Scale=0.02
    Sprite Rotation Mode=::NewEnumerator2
    Sprite Size Mode=::NewEnumerator3
    UsePositionOffset=false
    Write Material Random=true
   )
   &CameraOffset(Camera Offset Amount=2)
  ]
  particleUpdateScript [
   &ParticleState(Kill Particles When Lifetime Has Elapsed=false)
   &SkeletalMeshLocation(
    Bone / Socket Index=&SelectIntFromArray(Direct Array Index=$Particles.PartitionParticles.Partition, Int Selection Array=DataInterface() {propertyValues {IntData [1,2]}})
    MaskSpawnGroup=false
    Mesh Sampling Type=::NewEnumerator5
    Skeletal Mesh=$System.Skeletal Mesh
    Socket Sampling Mode=::NewEnumerator5
   )
   &ScaleColor(
    Scale Alpha=&Waveform("Waveform [1]"=::NewEnumerator2, "[1] Frequency"=1, "[1] Compound Frequency"=1, "[1] Amplitude Min/Max"=<0 0.5>)
    ScaleA=true
    ScaleRGB=true
   )
   &ScaleSpriteSize(
    Scale Sprite Size Mode=::NewEnumerator0
    Uniform Scale Factor=&Waveform("Waveform [1]"=::NewEnumerator2, "[1] Frequency"=20, "[1] Compound Frequency"=1, "[1] Amplitude Min/Max"=<0 1>)
   )
  ]
  rendererClasses [&NiagaraSpriteRendererProperties]
  renderers [{rendererClass=&NiagaraSpriteRendererProperties}]
 }
 BurstRing002 {
  emitterUpdateScript [
   &EmitterState(
    Inactive Response=::NewEnumerator1
    Life Cycle Mode=::NewEnumerator1
    Loop Behavior=::NewEnumerator1
    Loop Delay=0.25
    Loop Duration Mode=::NewEnumerator1
    UseLoopDelay=true
   )
   &SpawnRate(SpawnRate=&Lerp_Float(A=10, B=0, Alpha=$User.BuiltIn), Use Spawn Probability=true)
  ]
  particleSpawnScript [
   &PartitionParticles()
   &InitializeParticle(
    Color=&MultiplyLinearColor(
     Linear Color=<rgba 0.5 5 4.625 1>
     Scale Factor=&MakeVector4_FromLinearColor(LinearColor=$System.TODColorMultiplier)
    )
    Lifetime=0.5
    Mesh Scale Mode=::NewEnumerator3
    Mesh Uniform Scale=0.02
    Sprite Size Mode=::NewEnumerator3
    UsePositionOffset=true
    Write Material Random=true
   )
   &SkeletalMeshLocation(
    Bone / Socket Index=&SelectIntFromArray(Direct Array Index=$Particles.PartitionParticles.Partition, Int Selection Array=DataInterface() {propertyValues {IntData [1,2]}})
    MaskSpawnGroup=false
    Mesh Sampling Type=::NewEnumerator5
    Sampled Position Offset=&MakeVector(
     X=&Lerp_Float(
      A=0
      B=3
      Alpha=&FloatFromCurve(FloatCurve=DataInterface() {propertyValues {Curve {keys [{tangentMode=RCTM_User, time=0, value=0}, {tangentMode=RCTM_User, time=1, value=1}]}}})
     )
     Y=&Lerp_Float(
      A=0
      B=6
      Alpha=&FloatFromCurve(FloatCurve=DataInterface() {propertyValues {Curve {keys [{tangentMode=RCTM_User, time=0, value=0}, {tangentMode=RCTM_User, time=1, value=1}]}}})
     )
     Z=&Lerp_Float(
      A=0
      B=0
      Alpha=&FloatFromCurve(FloatCurve=DataInterface() {propertyValues {Curve {keys [{tangentMode=RCTM_User, time=0, value=0}, {tangentMode=RCTM_User, time=1, value=1}]}}})
     )
    )
    Skeletal Mesh=$System.Skeletal Mesh
    Socket Sampling Mode=::NewEnumerator5
   ) {module=SkeletalMeshLocation001}
  ]
  particleUpdateScript [
   &ParticleState()
   &SkeletalMeshLocation(
    Bone / Socket Index=&SelectIntFromArray(Direct Array Index=$Particles.PartitionParticles.Partition, Int Selection Array=DataInterface() {propertyValues {IntData [1,2]}})
    MaskSpawnGroup=false
    Mesh Sampling Type=::NewEnumerator5
    Sampled Position Offset=&MakeVector(
     X=&Lerp_Float(
      A=0
      B=3
      Alpha=&FloatFromCurve(FloatCurve=DataInterface() {propertyValues {Curve {keys [{tangentMode=RCTM_User, time=0, value=0}, {tangentMode=RCTM_User, time=1, value=1}]}}})
     )
     Y=&Lerp_Float(
      A=0
      B=6
      Alpha=&FloatFromCurve(FloatCurve=DataInterface() {propertyValues {Curve {keys [{tangentMode=RCTM_User, time=0, value=0}, {tangentMode=RCTM_User, time=1, value=1}]}}})
     )
     Z=&Lerp_Float(
      A=0
      B=0
      Alpha=&FloatFromCurve(FloatCurve=DataInterface() {propertyValues {Curve {keys [{tangentMode=RCTM_User, time=0, value=0}, {tangentMode=RCTM_User, time=1, value=1}]}}})
     )
    )
    Skeletal Mesh=$System.Skeletal Mesh
    Socket Sampling Mode=::NewEnumerator5
   )
   /SetVariables(Particles.MeshOrientation=&MultiplyQuaternion(
     Quaternion A=$Particles.SkeletalMeshLocation.SampledRotation
     Quaternion B=&MakeQuatFromAxisAngle(Axis=<1 0 0>, AngleInDegrees=90)
    )) {isSetParameters=true}
   &DynamicMaterialParameters(
    Index 0 Param 1=&MakeFloatFromVector(Vector=$Particles.Scale, Channel=::NewEnumerator0)
    Index 0 Param 2=2
    Index 0 Param 3=-0.3
    Index 0 Param 4=15
    Index 1 Param 2=1
    Index 1 Param 3=0.5
    Index 2 Param 1=2
    Index 2 Param 2=2
    Index 2 Param 3=1
    Module.Index0Float0Write=true
    Module.Index0Float1Write=true
    Module.Index0Float2Write=true
    Module.Index0Float3Write=true
    Module.Index1Float0Write=true
    Module.Index1Float1Write=true
    Module.Index1Float2Write=true
    Module.Index1Float3Write=true
    Module.Index2Float0Write=true
    Module.Index2Float1Write=true
    Module.Index2Float2Write=true
    Module.Index2Float3Write=true
    Write Parameter Index 1=::NewEnumerator1
    Write Parameter Index 2=::NewEnumerator1
   )
   &ScaleMeshSize(Scale Factor=&VectorFromFloat(Value=&FloatFromCurve(FloatCurve=DataInterface() {propertyValues {Curve {keys [
          {interpMode=RCIM_Cubic, leaveTangent=3.9642, tangentMode=RCTM_Break, time=0, value=0.5}
          {arriveTangent=0.297, interpMode=RCIM_Cubic, tangentMode=RCTM_Break, time=1, value=2.4}
         ]}}})))
   &ScaleColor(
    Scale Alpha=&FloatFromCurve(
     FloatCurve=DataInterface() {propertyValues {Curve {keys [
         {interpMode=RCIM_Cubic, tangentMode=RCTM_User, time=0, value=0}
         {arriveTangent=-1.4125, interpMode=RCIM_Cubic, leaveTangent=-1.4125, tangentMode=RCTM_User, time=0.4, value=1}
         {tangentMode=RCTM_User, time=1, value=0}
        ]}}}
     Scale Curve=0.5
    )
    Scale RGB=&VectorFromFloat(Value=&FloatFromCurve(FloatCurve=DataInterface() {propertyValues {Curve {keys [{tangentMode=RCTM_User, time=0, value=1}, {tangentMode=RCTM_User, time=1, value=1}]}}}))
    ScaleA=true
    ScaleRGB=true
   )
  ]
  rendererClasses [&NiagaraMeshRendererProperties]
  renderers [{rendererClass=&NiagaraMeshRendererProperties, rendererIndex=0}]
 }
 Chromatic_Spheres {
  emitterUpdateScript [
   &EmitterState(
    Inactive Response=::NewEnumerator1
    Life Cycle Mode=::NewEnumerator1
    Loop Behavior=::NewEnumerator1
    Loop Delay=$System.Delay
    Loop Duration Mode=::NewEnumerator1
    UseLoopDelay=true
   )
   &SpawnBurst_Instantaneous(Spawn Count=2, Use Spawn Probability=false, UseLoopCountLimit=false)
  ]
  particleSpawnScript [
   &PartitionParticles()
   &InitializeParticle(
    Color=&MergeHelper_ColorSwitch(0=<rgba 2 10 2.0471 1>, 1=<rgba 10 2 2 1>, 2=?, 3=?, Index=$Particles.PartitionParticles.Partition, Num=::NewEnumerator0)
    Mesh Scale Mode=::NewEnumerator3
    Mesh Uniform Scale=&Multiply_Float(A=0.225, B=$System.GlobalScale)
    Position Offset=$System.PositionOffset
    UsePositionOffset=true
    Write Material Random=true
   ) {module=InitializeParticle001}
   &InitialMeshOrientation(Mesh Orientation Mode=::NewEnumerator1, Rotation=<0 0.28 0>, Use Rotation Vector=true)
   &CameraOffset(Camera Offset Amount=&Lerp_Float(A=0, B=5, Alpha=$User.BuiltIn))
  ]
  particleUpdateScript [
   &ParticleState(Kill Particles When Lifetime Has Elapsed=false)
   &SkeletalMeshLocation(
    MaskSpawnGroup=false
    Mesh Sampling Type=::NewEnumerator5
    Position Sampling=::NewEnumerator2
    Skeletal Mesh=$System.Skeletal Mesh
    Socket Sampling Mode=::NewEnumerator5
   )
   &RotateAroundPoint(
    Derive the Rotation Axis=false
    Radius=2
    Rotation Axis:Vector3f=&MergeHelper_VectorSwitch(
     0=&Lerp_Vector(A=<1 -1 -1>, B=<-1 1 1>, Alpha=&Cosine(Normalized Angle=$Emitter.Age, Period=0.1, Scale=1, Bias=0))
     1=&Lerp_Vector(A=<-1 1 1>, B=<1 -1 -1>, Alpha=&Cosine(Normalized Angle=$Emitter.Age, Period=0.1, Scale=1, Bias=0))
     2=?
     3=?
     Index=$Particles.PartitionParticles.Partition
     Num=::NewEnumerator0
    )
    Rotation Center:NiagaraPosition=&Lerp_Vector(
     A=Linked() {linkedVariable {name=Particles.SkeletalMeshLocation.SampledPosition, type=NiagaraPosition}}
     B=Linked() {linkedVariable {name=Particles.Initial.Position, type=NiagaraPosition}}
     Alpha=$User.BuiltIn
    )
    Rotation Rate=10
   )
   &ScaleMeshSize(Scale Factor=&VectorFromFloat(Value=&FloatFromCurve(FloatCurve=DataInterface() {propertyValues {Curve {
         defaultValue=1
         keys [
          {arriveTangent=0.3976, interpMode=RCIM_Cubic, time=0, value=1}
          {arriveTangent=0.0057, interpMode=RCIM_Cubic, leaveTangent=0.0057, tangentMode=RCTM_User, time=0.28, value=1}
          <at 0.3 1.2 RCIM_Cubic>
          <at 0.4 1 RCIM_Cubic>
         ]
        }}})))
   &ScaleColor(
    Scale Alpha=&FloatFromCurve(
     FloatCurve=DataInterface() {propertyValues {Curve {
        defaultValue=1
        keys [
         {arriveTangent=0.3976, time=0, value=0}
         {arriveTangent=1.5034, interpMode=RCIM_Cubic, leaveTangent=40.2528, tangentMode=RCTM_Break, time=0.26, value=0}
         {arriveTangent=50.1869, interpMode=RCIM_Cubic, leaveTangent=-14.3, tangentMode=RCTM_Break, time=0.28, value=1.5}
         <at 0.4 0.5 RCIM_Cubic>
        ]
       }}}
     Scale Curve=&Multiply_Float(A=1, B=&FNTimeOfDayScalar(Night Scale=0.7, Morning Scale=0.85, Evening Scale=0.85))
    )
    ScaleA=true
    ScaleRGB=true
   )
   &DynamicMaterialParameters(
    Index 0 Param 3=&FloatFromCurve(FloatCurve=DataInterface() {propertyValues {Curve {keys [<at 0.5 0.4>,<at 1 1>]}}})
    Index 0 Param 4=&Lerp_Float(
     A=0.75
     B=&Lerp_Float(A=1, B=3, Alpha=$User.BuiltIn)
     Alpha=&Waveform("Waveform [1]"=::NewEnumerator2, "[1] Frequency"=0.4, "[1] Amplitude Min/Max"=<0 1>)
    )
    Module.Index0Float0Write=true
    Module.Index0Float1Write=true
    Module.Index0Float2Write=true
    Module.Index0Float3Write=true
   )
   &SolveForcesAndVelocity(Clamp Velocity=false, Limit Acceleration=false)
  ]
  rendererClasses [&NiagaraMeshRendererProperties]
  renderers [{rendererClass=&NiagaraMeshRendererProperties, rendererIndex=0}]
 }
 Gathering_Elements {
  emitterUpdateScript [
   &EmitterState(
    Inactive Response=::NewEnumerator1
    Life Cycle Mode=::NewEnumerator1
    Loop Behavior=::NewEnumerator1
    Loop Delay=$System.Delay
    Loop Duration=2
    UseLoopDelay=true
   )
   &SpawnBurst_Instantaneous(Spawn Count=5, Use Spawn Probability=false, UseLoopCountLimit=false)
  ]
  particleSpawnScript [
   &PartitionParticles(Number Of Partitions=5)
   /SetVariables(Particles.VisibilityTag=$Particles.PartitionParticles.Partition) {isSetParameters=true, module=SetVariables_#6}
   &InitializeParticle(
    Color=&MultiplyLinearColor(
     Linear Color=&MergeHelper_ColorSwitch(
      0=<rgba 0.1042 1.5017 2.5 1>
      1=<rgba 2.3304 3.3043 4 1>
      2=<rgba 0.0729 0.5365 1 1>
      3=<rgba 0.1029 0.3014 0.5 1>
      4=<rgba 0.6284 2.0856 5 1>
      Index=$Particles.PartitionParticles.Partition
      Num=::NewEnumerator3
     )
     Scale Factor=&MakeVector4_FromLinearColor(LinearColor=$System.TODColorMultiplier)
    )
    Lifetime=&MergeHelper_FloatSwitch(0=0.4, 1=0.4, 2=0.4, 3=0.4, 4=0.25, Index=$Particles.PartitionParticles.Partition, Num=::NewEnumerator3)
    Mesh Scale Mode=::NewEnumerator3
    Mesh Uniform Scale=&Multiply_Float(
     A=&MergeHelper_FloatSwitch(0=0, 1=0.4, 2=0.55, 3=0.34, 4=0, Index=$Particles.PartitionParticles.Partition, Num=::NewEnumerator3)
     B=$System.GlobalScale
    )
    Position Offset=$System.PositionOffset
    Sprite Rotation Mode=::NewEnumerator2
    Sprite Size Mode=::NewEnumerator3
    Uniform Sprite Size=&Multiply_Float(
     A=&MergeHelper_FloatSwitch(0=48, 1=0, 2=0, 3=0, 4=55, Index=$Particles.PartitionParticles.Partition, Num=::NewEnumerator3)
     B=$System.GlobalScale
    )
    UsePositionOffset=true
    Write Material Random=true
   )
   &CameraOffset(Camera Offset Amount=&MergeHelper_FloatSwitch(0=5, 1=5, 2=5, 3=5, 4=25, Index=$Particles.PartitionParticles.Partition, Num=::NewEnumerator3))
   &InitialMeshOrientation(
    Mesh Orientation Mode=::NewEnumerator2
    Rotation=&RandomRangeVector(Minimum=<-0.05 0 0>, Maximum=<0.05 0 0>)
    Use Rotation Vector=true
   )
  ]
  particleUpdateScript [
   &ParticleState()
   &SkeletalMeshLocation(
    MaskSpawnGroup=false
    Mesh Sampling Type=::NewEnumerator5
    Orientation Sampling=::NewEnumerator1
    Position Sampling=::NewEnumerator2
    Skeletal Mesh=$System.Skeletal Mesh
    Socket Sampling Mode=::NewEnumerator5
   )
   /SetVariables(Particles.Position=&Lerp_Vector(
     A=Linked() {linkedVariable {name=Particles.SkeletalMeshLocation.SampledPosition, type=NiagaraPosition}}
     B=Linked() {linkedVariable {name=Particles.Position, type=NiagaraPosition}}
     Alpha=$User.BuiltIn
    )) {isSetParameters=true, module=SetVariables_#7}
   &SpriteRotationRate()
   &ScaleSpriteSize(
    Scale Sprite Size Mode=::NewEnumerator0
    Uniform Scale Factor=&MergeHelper_FloatSwitch(
     0=&FloatFromCurve(FloatCurve=DataInterface() {propertyValues {Curve {keys [
          {arriveTangent=0.4801, interpMode=RCIM_Cubic, leaveTangent=0.4801, tangentMode=RCTM_User, time=0, value=0.3}
          {arriveTangent=3.752, interpMode=RCIM_Cubic, leaveTangent=5.5594, tangentMode=RCTM_Break, time=0.3, value=1}
          {arriveTangent=3.9949, interpMode=RCIM_Cubic, leaveTangent=3.9949, tangentMode=RCTM_User, time=0.6475, value=1.6025}
          {arriveTangent=0.0597, interpMode=RCIM_Cubic, leaveTangent=0.0597, tangentMode=RCTM_User, time=1, value=2.5}
         ]}}})
     1=0
     2=0
     3=0
     4=&FloatFromCurve(FloatCurve=DataInterface() {propertyValues {Curve {keys [
          {arriveTangent=-2.0515, interpMode=RCIM_Cubic, leaveTangent=-2.0515, tangentMode=RCTM_User, time=0, value=1}
          {arriveTangent=-0.202, interpMode=RCIM_Cubic, leaveTangent=-0.202, tangentMode=RCTM_User, time=1, value=0.2}
         ]}}})
     Index=$Particles.PartitionParticles.Partition
     Num=::NewEnumerator3
    )
   )
   &ScaleMeshSize(Scale Factor=&VectorFromFloat(Value=&MergeHelper_FloatSwitch(
      0=0
      1=&FloatFromCurve(FloatCurve=DataInterface() {propertyValues {Curve {
          defaultValue=0.15
          keys [{interpMode=RCIM_Cubic, tangentMode=RCTM_User, time=0, value=1}, {interpMode=RCIM_Cubic, tangentMode=RCTM_User, time=1, value=0}]
         }}})
      2=&FloatFromCurve(FloatCurve=DataInterface() {propertyValues {Curve {keys [
           {arriveTangent=-2.0161, interpMode=RCIM_Cubic, leaveTangent=-2.0161, tangentMode=RCTM_User, time=0, value=1}
           {arriveTangent=-0.3283, interpMode=RCIM_Cubic, leaveTangent=-0.3283, tangentMode=RCTM_User, time=0.6, value=0.2}
          ]}}})
      3=&FloatFromCurve(FloatCurve=DataInterface() {propertyValues {Curve {keys [
           {arriveTangent=-1.1673, interpMode=RCIM_Cubic, leaveTangent=-1.1673, tangentMode=RCTM_User, time=0, value=1}
           {arriveTangent=-0.3063, interpMode=RCIM_Cubic, leaveTangent=-0.3063, tangentMode=RCTM_User, time=1, value=0.3}
          ]}}})
      Index=$Particles.PartitionParticles.Partition
      Num=::NewEnumerator2
     )))
   &ScaleColor(
    Scale Alpha=&MergeHelper_FloatSwitch(
     0=&FloatFromCurve(FloatCurve=DataInterface() {propertyValues {Curve {keys [
          {arriveTangent=0.2601, interpMode=RCIM_Cubic, leaveTangent=0.2601, tangentMode=RCTM_User, time=0, value=0}
          {arriveTangent=1.7089, interpMode=RCIM_Cubic, leaveTangent=3.0471, tangentMode=RCTM_Break, time=0.5, value=0.4}
          {arriveTangent=-0.0028, interpMode=RCIM_Cubic, leaveTangent=-3.936, tangentMode=RCTM_Break, time=0.6, value=0.7}
          {arriveTangent=-3.2198, interpMode=RCIM_Cubic, leaveTangent=-3.2198, tangentMode=RCTM_User, time=0.7, value=0.35}
          {arriveTangent=0.0005, interpMode=RCIM_Cubic, leaveTangent=0.0005, tangentMode=RCTM_User, time=1, value=0}
         ]}}})
     1=&FloatFromCurve(
      FloatCurve=DataInterface() {propertyValues {Curve {keys [{arriveTangent=-2.4108, interpMode=RCIM_Cubic, leaveTangent=-2.4108, tangentMode=RCTM_User, time=0.3, value=1}, <at 1 0 RCIM_Cubic>]}}}
      Scale Curve=0.25
     )
     2=&FloatFromCurve(FloatCurve=DataInterface() {propertyValues {Curve {keys [
          {arriveTangent=0.0341, interpMode=RCIM_Cubic, leaveTangent=0.0341, tangentMode=RCTM_User, time=0, value=0}
          {arriveTangent=0.0847, interpMode=RCIM_Cubic, leaveTangent=-0.0684, tangentMode=RCTM_Break, time=0.4, value=1}
          {arriveTangent=0.0005, interpMode=RCIM_Cubic, leaveTangent=0.0005, tangentMode=RCTM_User, time=0.7, value=0}
         ]}}})
     3=&FloatFromCurve(
      FloatCurve=DataInterface() {propertyValues {Curve {keys [
          {arriveTangent=0.0068, interpMode=RCIM_Cubic, leaveTangent=0.0068, tangentMode=RCTM_User, time=0, value=0}
          {arriveTangent=-0.018, interpMode=RCIM_Cubic, leaveTangent=-0.018, tangentMode=RCTM_User, time=0.2, value=1}
          {arriveTangent=0.0005, interpMode=RCIM_Cubic, leaveTangent=0.0005, tangentMode=RCTM_User, time=0.5, value=0}
         ]}}}
      Scale Curve=0.3
     )
     4=&FloatFromCurve(
      FloatCurve=DataInterface() {propertyValues {Curve {keys [
          {arriveTangent=0.0195, interpMode=RCIM_Cubic, leaveTangent=0.0195, tangentMode=RCTM_User, time=0, value=0}
          <at 0.3 1 RCIM_Cubic>
          {interpMode=RCIM_Cubic, tangentMode=RCTM_User, time=1, value=0}
         ]}}}
      Scale Curve=0.8
     )
     Index=$Particles.PartitionParticles.Partition
     Num=::NewEnumerator3
    )
    Scale RGB=&MakeVectorFromLinearColorRGB(Color=<rgba 1 1 1 1>)
    ScaleA=true
    ScaleRGB=false
   )
   &DynamicMaterialParameters(
    Index 0 Param 1=1
    Index 0 Param 2=0.8802
    Index 0 Param 4=1
    Module.Index0Float0Write=true
    Module.Index0Float1Write=true
    Module.Index0Float2Write=true
    Module.Index0Float3Write=true
   )
   /SetVariables(Particles.DynPar_Caustics:Vector4f=&MakeVector4(
     X=1
     Y=&FloatFromCurve(FloatCurve=DataInterface() {propertyValues {Curve {keys [<at 0 1 RCIM_Cubic>,<at 1 0>]}}})
     Z=0
     W=0
    )) {isSetParameters=true}
  ]
  rendererClasses [&NiagaraSpriteRendererProperties, &NiagaraMeshRendererProperties]
  renderers [
   {rendererClass=&NiagaraSpriteRendererProperties}
   {rendererClass=&NiagaraMeshRendererProperties, rendererIndex=1}
   {rendererClass=&NiagaraMeshRendererProperties, rendererIndex=2}
   {rendererClass=&NiagaraMeshRendererProperties, rendererIndex=3}
   {rendererClass=&NiagaraSpriteRendererProperties, rendererIndex=4}
  ]
 }
 Idle_Sphere_Blue {
  emitterUpdateScript [
   &EmitterState(
    Inactive Response=::NewEnumerator1
    Life Cycle Mode=::NewEnumerator1
    Loop Behavior=::NewEnumerator1
    Loop Delay=$System.Delay
    Loop Duration Mode=::NewEnumerator1
    UseLoopDelay=true
   )
   &SpawnBurst_Instantaneous(Use Spawn Probability=false, UseLoopCountLimit=false)
  ]
  particleSpawnScript [
   &InitializeParticle(
    Color=<rgba 2.6176 6.945 10 1>
    Mesh Scale Mode=::NewEnumerator3
    Mesh Uniform Scale=&Multiply_Float(A=0.25, B=$System.GlobalScale)
    Position Offset=$System.PositionOffset
    UsePositionOffset=true
    Write Material Random=true
   ) {module=InitializeParticle001}
   &InitialMeshOrientation(Mesh Orientation Mode=::NewEnumerator1, Rotation=<0 0.28 0>, Use Rotation Vector=true)
   &CameraOffset(Camera Offset Amount=&Lerp_Float(A=0, B=15, Alpha=$User.BuiltIn))
  ]
  particleUpdateScript [
   &ParticleState(Kill Particles When Lifetime Has Elapsed=false)
   &SkeletalMeshLocation(
    MaskSpawnGroup=false
    Mesh Sampling Type=::NewEnumerator5
    Position Sampling=::NewEnumerator2
    Skeletal Mesh=$System.Skeletal Mesh
    Socket Sampling Mode=::NewEnumerator5
   )
   /SetVariables(Particles.Position=&Lerp_Vector(
     A=Linked() {linkedVariable {name=Particles.SkeletalMeshLocation.SampledPosition, type=NiagaraPosition}}
     B=Linked() {linkedVariable {name=Particles.Position, type=NiagaraPosition}}
     Alpha=$User.BuiltIn
    )) {isSetParameters=true, module=SetVariables_#7}
   &ScaleMeshSize(Scale Factor=&VectorFromFloat(Value=&FloatFromCurve(FloatCurve=DataInterface() {propertyValues {Curve {
         defaultValue=1
         keys [
          {arriveTangent=0.3976, interpMode=RCIM_Cubic, time=0, value=1}
          {arriveTangent=0.0057, interpMode=RCIM_Cubic, leaveTangent=0.0057, tangentMode=RCTM_User, time=0.28, value=1}
          <at 0.3 1.2 RCIM_Cubic>
          <at 0.4 1 RCIM_Cubic>
         ]
        }}})))
   &ScaleColor(
    Scale Alpha=&FloatFromCurve(FloatCurve=DataInterface() {propertyValues {Curve {
        defaultValue=1
        keys [
         {arriveTangent=0.3976, time=0, value=0}
         {arriveTangent=1.5034, interpMode=RCIM_Cubic, leaveTangent=40.2528, tangentMode=RCTM_Break, time=0.26, value=0}
         {arriveTangent=50.1869, interpMode=RCIM_Cubic, leaveTangent=-14.3, tangentMode=RCTM_Break, time=0.28, value=1.5}
         <at 0.4 0.5 RCIM_Cubic>
        ]
       }}})
    ScaleA=true
    ScaleRGB=true
   )
   &DynamicMaterialParameters(
    Index 0 Param 3=&FloatFromCurve(FloatCurve=DataInterface() {propertyValues {Curve {keys [<at 0.5 0.4>,<at 1 1>]}}})
    Index 0 Param 4=&Lerp_Float(A=1, B=1.5, Alpha=&Sine(Normalized Angle=$Emitter.Age, Period=1, Scale=1, Bias=0))
    Module.Index0Float0Write=true
    Module.Index0Float1Write=true
    Module.Index0Float2Write=true
    Module.Index0Float3Write=true
   )
   /SetVariables(
    Particles.DynPar_SoapySphere:Vector4f=&MakeVector4(X=0, Y=0, Z=1, W=0)
    Particles.SoapySphere_Color:LinearColor=&MultiplyLinearColor(
     Linear Color=<rgba 1 1 1 1>
     Scale Factor=&MakeVector4FromVectorAndFloat(
      "Vector (RGB)"=&MakeVectorFromLinearColorRGB(Color=<rgba 1 1 1 1>)
      "Float (Alpha)"=&FloatFromCurve(
       FloatCurve=DataInterface() {propertyValues {Curve {
          defaultValue=1
          keys [
           {arriveTangent=0.3976, time=0, value=0}
           {arriveTangent=1.5034, interpMode=RCIM_Cubic, leaveTangent=40.2528, tangentMode=RCTM_Break, time=0.26, value=0}
           {arriveTangent=50.1869, interpMode=RCIM_Cubic, leaveTangent=-14.3, tangentMode=RCTM_Break, time=0.28, value=1.5}
           <at 0.4 0.5 RCIM_Cubic>
          ]
         }}}
       Scale Curve=0.5
      )
     )
    )
   ) {isSetParameters=true}
  ]
  rendererClasses [&NiagaraMeshRendererProperties]
  renderers [
   {rendererClass=&NiagaraMeshRendererProperties, rendererIndex=0}
   {rendererClass=&NiagaraMeshRendererProperties, rendererIndex=1}
  ]
 }
 Looping_NonSphereElements {
  emitterUpdateScript [
   &EmitterState(
    Inactive Response=::NewEnumerator1
    Life Cycle Mode=::NewEnumerator1
    Loop Behavior=::NewEnumerator1
    Loop Delay=$System.Delay
    Loop Duration Mode=::NewEnumerator1
    UseLoopDelay=true
   )
   &SpawnBurst_Instantaneous(Spawn Count=8, Use Spawn Probability=false, UseLoopCountLimit=false)
  ]
  particleSpawnScript [
   &PartitionParticles(Number Of Partitions=8)
   /SetVariables(Particles.VisibilityTag=$Particles.PartitionParticles.Partition) {isSetParameters=true, module=SetVariables_#8}
   &InitializeParticle(
    Color=&MergeHelper_ColorSwitch(
     0=<rgba 0.0417 0.6007 1 1>
     1=<rgba 0.1529 0.4823 0.8116 0.3>
     2=<rgba 0.8521 5.731 10 1>
     3=<rgba 0.8 4.4 8 1>
     4=<rgba 0.8957 2.1232 3 1>
     5=<rgba 0.992 1.58 2 1>
     6=<rgba 2.0435 3.5217 5 1>
     7=<rgba 0.0729 0.6137 1 0.3>
     Index=$Particles.PartitionParticles.Partition
     Num=::NewEnumerator6
    )
    Mesh Scale Mode=::NewEnumerator3
    Mesh Uniform Scale=&Multiply_Float(
     A=&MergeHelper_FloatSwitch(0=0, 1=0, 2=0, 3=0, 4=0, 5=0, 6=0.2, 7=0.262, Index=$Particles.PartitionParticles.Partition, Num=::NewEnumerator6)
     B=$System.GlobalScale
    )
    Position Offset=$System.PositionOffset
    Sprite Rotation Angle=&MergeHelper_FloatSwitch(
     0=&RandomRangeFloat(Minimum=0, Maximum=1)
     1=&RandomRangeFloat(Minimum=0, Maximum=1)
     2=&RandomRangeFloat(Minimum=0, Maximum=1)
     3=0
     4=&RandomRangeFloat(Minimum=0, Maximum=1)
     5=&RandomRangeFloat(Minimum=0, Maximum=1)
     Index=$Particles.PartitionParticles.Partition
     Num=::NewEnumerator4
    )
    Sprite Rotation Mode=::NewEnumerator1
    Sprite Size=&Multiply_Vector2D_ByFloat(
     Float=$System.GlobalScale
     Vector2D=&MergeHelper_Vector2Switch(
      0=<60 60>
      1=<36 36>
      2=<15 15>
      3=<25 25>
      4=<35 35>
      5=<38 38>
      6=<0 0>
      7=<0 0>
      Index=$Particles.PartitionParticles.Partition
      Num=::NewEnumerator6
     )
    )
    Sprite Size Mode=::NewEnumerator4
    UsePositionOffset=true
    Write Material Random=true
   )
   &CameraOffset(Camera Offset Amount=&MergeHelper_FloatSwitch(0=5, 1=15, 2=5, 3=15, 4=0, 5=-10, 6=0, 7=0, Index=$Particles.PartitionParticles.Partition, Num=::NewEnumerator6))
   /SetVariables(Particles.LightExponent=4) {isSetParameters=true, module=SetVariables_#9}
   &InitialMeshOrientation(
    Mesh Orientation Mode=::NewEnumerator2
    Rotation=&RandomRangeVector(Minimum=<-0.05 0 0>, Maximum=<0.05 0 0>)
    Use Rotation Vector=true
   )
  ]
  particleUpdateScript [
   &ParticleState(Kill Particles When Lifetime Has Elapsed=false)
   &SkeletalMeshLocation(
    MaskSpawnGroup=false
    Mesh Sampling Type=::NewEnumerator5
    Orientation Sampling=::NewEnumerator1
    Position Sampling=::NewEnumerator2
    Skeletal Mesh=$System.Skeletal Mesh
    Socket Sampling Mode=::NewEnumerator5
   )
   /SetVariables(Particles.Position=&Lerp_Vector(
     A=Linked() {linkedVariable {name=Particles.SkeletalMeshLocation.SampledPosition, type=NiagaraPosition}}
     B=Linked() {linkedVariable {name=Particles.Position, type=NiagaraPosition}}
     Alpha=$User.BuiltIn
    )) {isSetParameters=true, module=SetVariables_#7}
   &SpriteRotationRate(Rotation Rate=5)
   &ScaleSpriteSize(Uniform Curve Sprite Scale=DataInterface() {propertyValues {Curve {
       defaultValue=1
       keys [
        {arriveTangent=0.3976, interpMode=RCIM_Cubic, time=0, value=0}
        {arriveTangent=0.0057, interpMode=RCIM_Cubic, leaveTangent=0.0057, tangentMode=RCTM_User, time=0.28, value=0}
        {interpMode=RCIM_Cubic, leaveTangent=8.3333, time=0.3, value=1}
       ]
      }}}) {module=ScaleSpriteSize001}
   &ScaleSpriteSize(
    Scale Sprite Size Mode=::NewEnumerator0
    Uniform Scale Factor=&Waveform(
     "Waveform [1]"=::NewEnumerator2
     "[1] Frequency"=0.5
     "[1] Compound Frequency"=1
     "[1] Amplitude Min/Max"=<0.9 1.2>
     "[1] Pulse Duration"=0.5
    )
   )
   &ScaleMeshSize(Scale Factor=&VectorFromFloat(Value=&FloatFromCurve(FloatCurve=DataInterface() {propertyValues {Curve {
         defaultValue=1
         keys [
          {arriveTangent=0.3976, interpMode=RCIM_Cubic, time=0, value=0}
          {arriveTangent=0.0057, interpMode=RCIM_Cubic, leaveTangent=0.0057, tangentMode=RCTM_User, time=0.28, value=0}
          {interpMode=RCIM_Cubic, leaveTangent=8.3333, time=0.3, value=1}
         ]
        }}}))) {module=ScaleMeshSize001}
   &ScaleColor(
    Scale Alpha=&Multiply_Float(
     A=&FloatFromCurve(FloatCurve=DataInterface() {propertyValues {Curve {
         defaultValue=1
         keys [
          {arriveTangent=0.3976, time=0, value=0}
          {arriveTangent=1.5034, interpMode=RCIM_Cubic, leaveTangent=40.2528, tangentMode=RCTM_Break, time=0.26, value=0}
          {arriveTangent=50.1869, interpMode=RCIM_Cubic, leaveTangent=-14.3, tangentMode=RCTM_Break, time=0.28, value=1.5}
          <at 0.4 0.5 RCIM_Cubic>
         ]
        }}})
     B=&MakeFloatFromLinearColor(LinearColor=$System.TODColorMultiplier, Channel:ENiagaraFloatFromLinearColorOptions=::NewEnumerator3)
    )
    Scale RGB=&MakeVectorFromLinearColorRGB(Color=$System.TODColorMultiplier)
    ScaleA=true
    ScaleRGB=true
   )
   &DynamicMaterialParameters(
    Index 0 Param 1=1
    Index 0 Param 2=0.85
    Index 0 Param 4=1
    Module.Index0Float0Write=true
    Module.Index0Float1Write=true
    Module.Index0Float2Write=true
    Module.Index0Float3Write=true
   )
   /SetVariables(Particles.DynPar_BasicGlow:Vector4f=&MakeVector4(X=2, Y=0.2, Z=0, W=0)) {isSetParameters=true, module=SetVariables_#10}
   /SetVariables(
    Particles.LightColor=&Multiply_LinearColorByFloat(
     Linear Color=$Particles.Color
     Float=&Waveform("Waveform [1]"=::NewEnumerator2, "[1] Frequency"=0.5, "[1] Compound Frequency"=1, "[1] Amplitude Min/Max"=<0 1.2>)
    )
    Particles.LightRadius=&Multiply_Float(
     A=&Multiply_Float(A=60, B=$System.GlobalScale)
     B=&FloatFromCurve(FloatCurve=DataInterface() {propertyValues {Curve {
         defaultValue=1
         keys [
          {arriveTangent=0.3976, interpMode=RCIM_Cubic, time=0, value=0}
          {arriveTangent=0.0057, interpMode=RCIM_Cubic, leaveTangent=0.0057, tangentMode=RCTM_User, time=0.28, value=0}
          {interpMode=RCIM_Cubic, leaveTangent=8.3333, time=0.3, value=1}
         ]
        }}})
    )
   ) {isSetParameters=true, module=SetVariables_#11}
   /SetVariables(Particles.DynPar Caustic:Vector4f=&MakeVector4(X=3, Y=1, Z=0, W=0)) {isSetParameters=true, module=SetVariables_#12}
   /SetVariables(Particles.DecalColor:LinearColor=&Multiply_LinearColorByFloat(
     Linear Color=$Particles.Color
     Float=&Waveform("Waveform [1]"=::NewEnumerator2, "[1] Frequency"=0.5, "[1] Compound Frequency"=1, "[1] Amplitude Min/Max"=<0.6 1.2>)
    )) {isSetParameters=true, module=SetVariables_#13}
  ]
  rendererClasses [
   &NiagaraSpriteRendererProperties
   &NiagaraLightRendererProperties
   &NiagaraMeshRendererProperties
  ]
  renderers [
   {rendererClass=&NiagaraSpriteRendererProperties}
   {rendererClass=&NiagaraSpriteRendererProperties, rendererIndex=1}
   {rendererClass=&NiagaraLightRendererProperties, rendererIndex=2}
   {rendererClass=&NiagaraSpriteRendererProperties, rendererIndex=3}
   {rendererClass=&NiagaraSpriteRendererProperties, rendererIndex=4}
   {rendererClass=&NiagaraSpriteRendererProperties, rendererIndex=5}
   {rendererClass=&NiagaraMeshRendererProperties, rendererIndex=6}
   {rendererClass=&NiagaraMeshRendererProperties, rendererIndex=7}
  ]
 }
 Looping_NonSphereElements001 {
  emitterUpdateScript [
   &EmitterState(
    Inactive Response=::NewEnumerator1
    Life Cycle Mode=::NewEnumerator1
    Loop Behavior=::NewEnumerator1
    Loop Delay=0.2
    Loop Duration Mode=::NewEnumerator1
    UseLoopDelay=true
   )
   &SpawnBurst_Instantaneous(
    Spawn Count=&MakeIntFromFloat(Float=&Lerp_Float(A=4, B=0, Alpha=$User.BuiltIn))
    Use Spawn Probability=false
    UseLoopCountLimit=false
   )
  ]
  particleSpawnScript [
   &PartitionParticles(Number Of Partitions=4)
   /SetVariables(Particles.VisibilityTag=$Particles.PartitionParticles.Partition) {isSetParameters=true, module=SetVariables_#8}
   &InitializeParticle(
    Color=&MergeHelper_ColorSwitch(
     0=<rgba 0.0417 0.6007 1 1>
     1=<rgba 0.1652 0.6522 1 0.3>
     2=<rgba 0.8521 5.731 10 1>
     3=<rgba 0.8 4.4 8 1>
     Index=$Particles.PartitionParticles.Partition
     Num=::NewEnumerator2
    )
    Mesh Scale=<0.05 0.1 0.1>
    Mesh Scale Mode=::NewEnumerator4
    Position Offset=$System.PositionOffset
    Sprite Rotation Angle=&MergeHelper_FloatSwitch(
     0=&RandomRangeFloat(Minimum=0, Maximum=1)
     1=&RandomRangeFloat(Minimum=0, Maximum=1)
     2=&RandomRangeFloat(Minimum=0, Maximum=1)
     3=0
     4=&RandomRangeFloat(Minimum=0, Maximum=1)
     5=&RandomRangeFloat(Minimum=0, Maximum=1)
     Index=$Particles.PartitionParticles.Partition
     Num=::NewEnumerator4
    )
    Sprite Rotation Mode=::NewEnumerator1
    Sprite Size Mode=::NewEnumerator3
    Uniform Sprite Size=&MakeFloatFromVector2d(Vector 2D=&Multiply_Vector2D_ByFloat(
      Float=$System.GlobalScale
      Vector2D=&MergeHelper_Vector2Switch(0=<60 60>, 1=<36 36>, 2=<15 15>, 3=<25 25>, 4=?, Index=$Particles.PartitionParticles.Partition, Num=::NewEnumerator2)
     ))
    UsePositionOffset=true
    Write Material Random=true
   )
   &CameraOffset(Camera Offset Amount=&MergeHelper_FloatSwitch(0=5, 1=5, 2=5, 3=15, 4=0, 5=-10, 6=0, 7=0, Index=$Particles.PartitionParticles.Partition, Num=::NewEnumerator6))
   /SetVariables(Particles.LightExponent=4) {isSetParameters=true, module=SetVariables_#9}
   &InitialMeshOrientation(
    Mesh Axis To Orient=<-0.3349 1.5 0>
    Mesh Orientation Mode=::NewEnumerator2
    Rotation=&RandomRangeVector(Minimum=<-0.05 0 0>, Maximum=<0 0 0>)
    Use Rotation Vector=true
   )
  ]
  particleUpdateScript [
   &ParticleState(Kill Particles When Lifetime Has Elapsed=false)
   &SkeletalMeshLocation(
    Bone / Socket Index=3
    MaskSpawnGroup=false
    Mesh Sampling Type=::NewEnumerator5
    Position Sampling=::NewEnumerator2
    Skeletal Mesh=$System.Skeletal Mesh
    Socket Sampling Mode=::NewEnumerator5
   )
   /SetVariables(Particles.Position=&Lerp_Vector(
     A=Linked() {linkedVariable {name=Particles.SkeletalMeshLocation.SampledPosition, type=NiagaraPosition}}
     B=Linked() {linkedVariable {name=Particles.Position, type=NiagaraPosition}}
     Alpha=$User.BuiltIn
    )) {isSetParameters=true, module=SetVariables_#7}
   &SpriteRotationRate(Rotation Rate=5)
   &ScaleSpriteSize(Uniform Curve Sprite Scale=DataInterface() {propertyValues {Curve {
       defaultValue=1
       keys [{arriveTangent=0.3976, interpMode=RCIM_Cubic, time=0, value=0}, {interpMode=RCIM_Cubic, leaveTangent=8.3333, time=0.3, value=1}]
      }}}) {module=ScaleSpriteSize001}
   &ScaleSpriteSize(
    Scale Sprite Size Mode=::NewEnumerator0
    Uniform Scale Factor=&Waveform(
     "Waveform [1]"=::NewEnumerator2
     "[1] Frequency"=0.5
     "[1] Compound Frequency"=1
     "[1] Amplitude Min/Max"=<0.9 1.2>
     "[1] Pulse Duration"=0.5
    )
   )
   &ScaleMeshSize(Scale Factor=&VectorFromFloat(Value=&FloatFromCurve(FloatCurve=DataInterface() {propertyValues {Curve {
         defaultValue=1
         keys [
          {arriveTangent=0.3976, interpMode=RCIM_Cubic, time=0, value=0}
          {arriveTangent=0.0057, interpMode=RCIM_Cubic, leaveTangent=0.0057, tangentMode=RCTM_User, time=0.28, value=0}
          {interpMode=RCIM_Cubic, leaveTangent=8.3333, time=0.3, value=1}
         ]
        }}}))) {module=ScaleMeshSize001}
   &ScaleColor(
    Scale Alpha=&Multiply_Float(
     A=&FloatFromCurve(FloatCurve=DataInterface() {propertyValues {Curve {
         defaultValue=1
         keys [
          {arriveTangent=0.3976, time=0, value=0}
          {arriveTangent=1.5034, interpMode=RCIM_Cubic, leaveTangent=40.2528, tangentMode=RCTM_Break, time=0.26, value=0.5}
          {arriveTangent=50.1869, interpMode=RCIM_Cubic, leaveTangent=-14.3, tangentMode=RCTM_Break, time=0.28, value=1.5}
          <at 0.4 0.5 RCIM_Cubic>
         ]
        }}})
     B=&MakeFloatFromLinearColor(LinearColor=$System.TODColorMultiplier, Channel:ENiagaraFloatFromLinearColorOptions=::NewEnumerator3)
    )
    Scale RGB=&MakeVectorFromLinearColorRGB(Color=$System.TODColorMultiplier)
    ScaleA=true
    ScaleRGB=true
   )
   &DynamicMaterialParameters(
    Index 0 Param 1=1
    Index 0 Param 2=0.85
    Index 0 Param 4=1
    Module.Index0Float0Write=true
    Module.Index0Float1Write=true
    Module.Index0Float2Write=true
    Module.Index0Float3Write=true
   )
   /SetVariables(Particles.DynPar_BasicGlow:Vector4f=&MakeVector4(X=2, Y=0.2, Z=0, W=0)) {isSetParameters=true, module=SetVariables_#10}
   /SetVariables(
    Particles.LightColor=&Multiply_LinearColorByFloat(
     Linear Color=$Particles.Color
     Float=&Waveform("Waveform [1]"=::NewEnumerator2, "[1] Frequency"=0.5, "[1] Compound Frequency"=1, "[1] Amplitude Min/Max"=<0 1.2>)
    )
    Particles.LightRadius=&Multiply_Float(
     A=&Multiply_Float(A=60, B=$System.GlobalScale)
     B=&FloatFromCurve(FloatCurve=DataInterface() {propertyValues {Curve {
         defaultValue=1
         keys [
          {arriveTangent=0.3976, interpMode=RCIM_Cubic, time=0, value=0}
          {arriveTangent=0.0057, interpMode=RCIM_Cubic, leaveTangent=0.0057, tangentMode=RCTM_User, time=0.28, value=0}
          {interpMode=RCIM_Cubic, leaveTangent=8.3333, time=0.3, value=1}
         ]
        }}})
    )
   ) {isSetParameters=true, module=SetVariables_#11}
   /SetVariables(Particles.DynPar Decal:Vector4f=&MakeVector4(X=1.5, Y=1.3452, Z=0, W=0)) {isSetParameters=true, module=SetVariables_#12}
   /SetVariables(Particles.DecalColor:LinearColor=&Multiply_LinearColorByFloat(
     Linear Color=$Particles.Color
     Float=&Waveform("Waveform [1]"=::NewEnumerator2, "[1] Frequency"=0.5, "[1] Compound Frequency"=1, "[1] Amplitude Min/Max"=<0.6 1.2>)
    )) {isSetParameters=true, module=SetVariables_#13}
  ]
  rendererClasses [&NiagaraSpriteRendererProperties]
  renderers [
   {rendererClass=&NiagaraSpriteRendererProperties}
   {rendererClass=&NiagaraSpriteRendererProperties, rendererIndex=1}
   {rendererClass=&NiagaraSpriteRendererProperties, rendererIndex=2}
  ]
 }
 center001 {
  emitterSpawnScript [/SetVariables(Emitter.WPO Sine Speed=1.5) {isSetParameters=true}]
  emitterUpdateScript [
   &EmitterState(
    Inactive Response=::NewEnumerator1
    Life Cycle Mode=::NewEnumerator1
    Loop Behavior=::NewEnumerator1
    Loop Delay=0.13
    Loop Duration Mode=::NewEnumerator1
    UseLoopDelay=true
   )
   &SpawnBurst_Instantaneous(
    Spawn Count=&MakeIntFromFloat(Float=&Lerp_Float(A=1, B=0, Alpha=$User.BuiltIn))
    Use Spawn Probability=false
    UseLoopCountLimit=false
   )
  ]
  particleSpawnScript [
   &InitializeParticle(
    Color=&MultiplyLinearColor(
     Linear Color=<rgba 0.2188 0.6745 1 1>
     Scale Factor=&MakeVector4_FromLinearColor(LinearColor=$System.TODColorMultiplier)
    )
    Mesh Scale=&Multiply_VectorByFloat(Vector=<0.4 0.2 0.2>, Float=0.8)
    Mesh Scale Mode=::NewEnumerator4
    UsePositionOffset=false
    Write Material Random=true
   )
   &InitialMeshOrientation(Mesh Orientation Mode=::NewEnumerator2, Rotation=<0 -0.25 0>, Use Rotation Vector=true)
  ]
  particleUpdateScript [
   &ParticleState(Kill Particles When Lifetime Has Elapsed=false)
   &ScaleMeshSize(Scale Factor=&MakeVector(
     X=&FloatFromCurve(FloatCurve=DataInterface() {propertyValues {Curve {keys [
          <at 0 1 RCIM_Cubic>
          {arriveTangent=-1.6667, interpMode=RCIM_Cubic, leaveTangent=-1.6667, time=0.2, value=0.5}
          {arriveTangent=2, interpMode=RCIM_Cubic, leaveTangent=2, time=0.3, value=0.5}
          <at 0.35 0.8 RCIM_Cubic>
         ]}}})
     Y=1
     Z=&FloatFromCurve(FloatCurve=DataInterface() {propertyValues {Curve {keys [
          <at 0 0 RCIM_Cubic>
          {arriveTangent=6.6667, interpMode=RCIM_Cubic, leaveTangent=6.6667, time=0.2, value=2}
          {arriveTangent=-6.6667, interpMode=RCIM_Cubic, leaveTangent=-6.6667, time=0.3, value=2}
          <at 0.35 1 RCIM_Cubic>
         ]}}})
    ))
   &SkeletalMeshLocation(
    Bone / Socket Index=3
    MaskSpawnGroup=false
    Mesh Sampling Type=::NewEnumerator5
    Position Sampling=::NewEnumerator2
    Sampled Position Offset=<0 0 -0.5>
    Skeletal Mesh=$System.Skeletal Mesh
    Socket Sampling Mode=::NewEnumerator5
   )
   /SetVariables(Particles.Position=&Lerp_Vector(
     A=Linked() {linkedVariable {name=Particles.SkeletalMeshLocation.SampledPosition, type=NiagaraPosition}}
     B=Linked() {linkedVariable {name=Particles.Position, type=NiagaraPosition}}
     Alpha=$User.BuiltIn
    )) {isSetParameters=true, module=SetVariables_#7}
   /SetVariables(Particles.Dark color:LinearColor=&MakeLinearColorFromVectorAndFloat(
     "Vector (RGB)"=&Multiply_Vector(
      A=&MakeVectorFromLinearColorRGB(Color=<rgba 0.0616 0.2894 0.4522 0.7>)
      B=&MakeVectorFromLinearColorRGB(Color=$System.TODColorMultiplier)
     )
     "Float (Alpha)"=0.7
    )) {isSetParameters=true}
   /SetVariables(Particles.Edge color:LinearColor=&MakeLinearColorFromVectorAndFloat(
     "Vector (RGB)"=&Multiply_Vector(
      A=&MakeVectorFromLinearColorRGB(Color=<rgba 0.7145 6.5953 10 1>)
      B=&MakeVectorFromLinearColorRGB(Color=$System.TODColorMultiplier)
     )
     "Float (Alpha)"=1
    )) {isSetParameters=true}
   &DynamicMaterialParameters(
    Index 0 Param 2=&FloatFromCurve(FloatCurve=DataInterface() {propertyValues {Curve {keys [<at 0 0 RCIM_Cubic>,<at 1 6 RCIM_Cubic>]}}})
    Module.Index0Float0Write=true
    Module.Index0Float1Write=true
    Module.Index0Float2Write=true
    Module.Index0Float3Write=true
   )
  ]
  rendererClasses [&NiagaraMeshRendererProperties]
  renderers [
   {rendererClass=&NiagaraMeshRendererProperties, rendererIndex=0}
   {rendererClass=&NiagaraMeshRendererProperties, rendererIndex=1}
   {rendererClass=&NiagaraMeshRendererProperties, rendererIndex=2}
  ]
 }
}
summary {
 systemName=NS_Pickaxe_CampWrench_Idle_Mesh
 emitters [
  {
   emitterName=Looping_NonSphereElements
   rendererClasses [
    &NiagaraSpriteRendererProperties
    &NiagaraLightRendererProperties
    &NiagaraMeshRendererProperties
   ]
  }
  {
   emitterName=Gathering_Elements
   rendererClasses [&NiagaraSpriteRendererProperties, &NiagaraMeshRendererProperties]
  }
  {emitterName=BurstElements, rendererClasses [&NiagaraSpriteRendererProperties]}
  {emitterName=Chromatic_Spheres, rendererClasses [&NiagaraMeshRendererProperties]}
  {emitterName=Looping_NonSphereElements001, rendererClasses [&NiagaraSpriteRendererProperties]}
  {emitterName=BurstRing002, rendererClasses [&NiagaraMeshRendererProperties]}
  {emitterName=center001, rendererClasses [&NiagaraMeshRendererProperties]}
  {emitterName=Idle_Sphere_Blue, rendererClasses [&NiagaraMeshRendererProperties]}
  {emitterName=BurstRing, rendererClasses [&NiagaraSpriteRendererProperties]}
 ]
 userVariables [
  User.bIsFrontEnd {type=NiagaraBool}
  User.bIsFrontEndPreview {type=NiagaraBool}
  User.BuiltIn {type=NiagaraFloat, defaultValue=0, description=""}
 ]
}
