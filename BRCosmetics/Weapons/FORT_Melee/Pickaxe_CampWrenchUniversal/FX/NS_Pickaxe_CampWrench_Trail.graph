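# Niagara swing-trail effect for the CampWrench pickaxe, made of two ribbon emitters that follow the skeletal mesh through AnimTrail modules.
# BG_WaveTrail is a half-opacity layer tinted by a blue color curve, while IridiscentTrail is a brighter, shorter-lived pale cyan layer dimmed at night.
# Both fade alpha over life scaled by a system TrailFade value, and User.BuiltIn blends the trail width.
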NiagaraSystem /BRCosmetics/Weapons/FORT_Melee/Pickaxe_CampWrenchUniversal/FX/NS_Pickaxe_CampWrench_Trail
dependencies {
 usedDataInterfaces [&NiagaraDataInterfaceCurve, &NiagaraDataInterfaceColorCurve, &NiagaraDataInterfaceSkeletalMesh]
 usedDynamicInputs [&self:SectionIndexFloatSelector]
 usedModules [
  &self:SystemSpawnScript.SystemScriptSource.SystemScriptGraph.NiagaraNodeAssignment_0.SetVariables_#1
  &SystemState
  &self:SystemSpawnScript.SystemScriptSource.SystemScriptGraph.NiagaraNodeAssignment_2.SetVariables_#2
 ]
}
emitters {
 BG_WaveTrail {
  emitterUpdateScript [
   &EmitterState()
   &FN_SpawnAnimTrail(Spawn Interval=0, Switch on ENiagara_AnimTrail_Source:ENiagara_AnimTrail_Source=::NewEnumerator1)
  ]
  particleSpawnScript [
   &FN_InitializeAnimTrailBoneIndecies(Second Bone or Socket Index=0)
   &FN_InitializeAnimTrail(
    AnimTrailWidth=&Lerp_Float(A=20, B=25, Alpha=$User.BuiltIn)
    Color=<rgba 1 1 1 0.5>
    Lifetime=0.25
    SkeletalMesh=$System.Skeletal Mesh
   )
  ]
  particleUpdateScript [
   &SolveForcesAndVelocity(Clamp Velocity=false, Limit Acceleration=false)
   &FN_UpdateAnimTrail(SkeletalMesh=$System.Skeletal Mesh) {module=FN_UpdateAnimTrail001}
   &ScaleColor(
    Scale Alpha=&FloatFromCurve(
     FloatCurve=DataInterface() {propertyValues {Curve {keys [
         {arriveTangent=-2.465, interpMode=RCIM_Cubic, leaveTangent=-2.465, tangentMode=RCTM_User, time=0, value=1}
         {interpMode=RCIM_Cubic, tangentMode=RCTM_User, time=1, value=0}
        ]}}}
     Scale Curve=$System.TrailFade
    )
    Scale RGB=&MakeVectorFromLinearColorRGB(Color=&ColorFromCurve(ColorCurve=DataInterface() {propertyValues {
        AlphaCurve {keys [<at 0 1 RCIM_Cubic>,<at 1 0 RCIM_Cubic>]}
        BlueCurve {keys [<at 0 1.5 RCIM_Cubic>,<at 0.8009 0.6841 RCIM_Cubic>]}
        GreenCurve {keys [<at 0 1.0391 RCIM_Cubic>,<at 0.8009 0.4313 RCIM_Cubic>]}
        RedCurve {keys [<at 0 0.1174 RCIM_Cubic>,<at 0.8009 0.0775 RCIM_Cubic>]}
       }}))
    ScaleA=true
    ScaleRGB=true
   )
  ]
  rendererClasses [&NiagaraRibbonRendererProperties]
  renderers [{rendererClass=&NiagaraRibbonRendererProperties}]
 }
 IridiscentTrail {
  emitterUpdateScript [
   &EmitterState()
   &FN_SpawnAnimTrail(Spawn Interval=0, Switch on ENiagara_AnimTrail_Source:ENiagara_AnimTrail_Source=::NewEnumerator1)
  ]
  particleSpawnScript [
   &FN_InitializeAnimTrailBoneIndecies(Second Bone or Socket Index=0)
   &FN_InitializeAnimTrail(
    AnimTrailWidth=&Lerp_Float(A=20, B=25, Alpha=$User.BuiltIn)
    Color=<rgba 1.9913 2.5797 3 1>
    Lifetime=0.2
    SkeletalMesh=$System.Skeletal Mesh
   )
  ]
  particleUpdateScript [
   &SolveForcesAndVelocity(Clamp Velocity=false, Limit Acceleration=false)
   &FN_UpdateAnimTrail(SkeletalMesh=$System.Skeletal Mesh) {module=FN_UpdateAnimTrail001}
   &ScaleColor(
    Scale Alpha=&FloatFromCurve(
     FloatCurve=DataInterface() {propertyValues {Curve {keys [
         {arriveTangent=-2.465, interpMode=RCIM_Cubic, leaveTangent=-2.465, tangentMode=RCTM_User, time=0, value=1}
         {interpMode=RCIM_Cubic, tangentMode=RCTM_User, time=1, value=0}
        ]}}}
     Scale Curve=&Multiply_Float(A=&FNTimeOfDayScalar(Night Scale=0.5, Morning Scale=1, Evening Scale=1), B=$System.TrailFade)
    )
    ScaleA=true
    ScaleRGB=true
   )
  ]
  rendererClasses [&NiagaraRibbonRendererProperties]
  renderers [{rendererClass=&NiagaraRibbonRendererProperties}]
 }
}
summary {
 systemName=NS_Pickaxe_CampWrench_Trail
 emitters [
  {emitterName=IridiscentTrail, rendererClasses [&NiagaraRibbonRendererProperties]}
  {emitterName=BG_WaveTrail, rendererClasses [&NiagaraRibbonRendererProperties]}
 ]
 userVariables [
  User.BuiltIn {type=NiagaraFloat, defaultValue=0, description=""}
  User.DebugSpeed {type=Vector3f, description="", defaultValue {struct=Vector3f, value=<0 0 0>}}
  User.RibbonIDOffset {type=NiagaraInt32}
  User.SwingSectionIndex {type=NiagaraInt32}
 ]
}
